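; Zero Point actor for after the Cattus/Doggus event that visibly destabilizes as the PCVD calendar event progresses.
; On BeginPlay the server hides it if the Cattus parameter collection flags it, revealing it once BP_CattusDoggus_Scripting's ball sequence triggers.
; Once playlist data is ready, the server converts active event time into Progress.
; RepNotifies apply visibility and collision, and drive the Nexus material collection, crack and wave meshes, tiered particles and ambient audio.
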(event EventBeginPlay
  (Networking|SwitchHasAuthority
    (:Authority
      (Development|PrintString "Cattus- ZeroPoint - BeginPlay" false)
      (Utilities|FlowControl|Sequence
        (:then_0
          (bind _returnvalue (Rendering|Material|GetScalarParameterValue "/Game/Athena/Environments/Cattus/Materials/MPC_CattusPOI.MPC_CattusPOI" "ZeroPointHidden" :!class /Script/Engine.KismetMaterialLibrary))
          (if (== _returnvalue 1.0)
            (Development|PrintString "Cattus- ZeroPoint - Hidden by MCP. Waiting for animation to trigger it" false)
            (Variables|Default|SetHideActor true)
            (bind _outactors (Actor|GetAllActorsOfClass "/Figment_S09/Athena/Prototype/Blueprints/Cattus_9_4/BP_CattusDoggus_Scripting.BP_CattusDoggus_Scripting_C"))
            (bind _output_get (Variables|Default|SetScriptingActor (Utilities|Array|Get(acopy) _outactors)))
            (Utilities|IsValid _output_get
              (:Is%20Valid
                (if (Math|Boolean|Equal(Boolean) (Variables|/Figment_S09/Athena/Prototype/Blueprints/Cattus_9_4/BP_CattusDoggus_Scripting.BP_CattusDoggus_Scripting_C|GetBallTriggered (Variables|Default|GetScriptingActor)) true)
                  (Variables|Default|SetHideActor false)
                  (Development|PrintString "Cattus- ZeroPoint - Already triggered by sequence. Unhiding it" false)
                  (else
                    (Default|BindEventtoOnBallTriggeredfromSequence (Variables|Default|GetScriptingActor) (delegate Custom|BallTriggered)))))
              (:Is%20Not%20Valid
                (Development|PrintString "Error: Cattus- ZeroPoint - Couldn't find CattusDoggusScripting" false)))))
        (:then_1
          (Class|FortAsyncActionCurrentPlaylistReady|PlaylistDataReadyAsync
            (:OnReady
              (bind (_biseventactive _timesincebegin _timeuntilend _timespanratio) (CalendarEvents|GetCalendarEventInformation "PCVD"))
              (if _biseventactive
                (bind _output_get_2 (Variables|Default|SetProgress (Math|Float|Clamp(Float) (Math|Float|NormalizetoRange _timespanratio 0.0 0.85))))
                (Development|PrintString (Utilities|String|Append "Cattus- Zero Point - PCVD Event Is Active. Progress: " (Utilities|String|ToString(Float) _output_get_2)) false)))))))))

(event Custom|BallTriggered
  (Variables|Default|SetHideActor false)
  (Development|PrintString "Cattus- ZeroPoint - Triggered by sequence. Unhiding it" false))

(fn OnRep_HideActor
  (bind _hideactor (Variables|Default|GetHideActor))
  (Rendering|SetActorHiddenInGame _hideactor)
  (Collision|SetActorEnableCollision (not _hideactor)))

(fn OnRep_Progress
  (CallFunction|UpdateProgress))

(fn UpdateProgress
  (bind _progress (Variables|Default|GetProgress))
  (Rendering|Material|SetScalarParameterValue "/Game/Athena/Environments/Nexus/Materials/FXshader/MPC_Nexus.MPC_Nexus" "Destabilize" (Math|Float|Clamp(Float) (Variables|Default|GetProgress)) :!class /Script/Engine.KismetMaterialLibrary)
  (Rendering|SetVisibility (Variables|Default|GetWave2) (> _progress 0.05) false)
  (Rendering|Material|SetScalarParameterValueonMaterials (Variables|Default|GetSM_Nexus_CracksRays) "PushDistance" (Math|Float|MapRangeClamped _progress 0.1 0.7 1.2e+03 7e+03))
  (Rendering|SetVisibility (Variables|Default|GetSM_Nexus_CracksRaysDouble) (> _progress 0.8) false)
  (Effects|Components|ParticleSystem|SetTemplate (Variables|Default|GetP_ZreoPoint_Destabilize) (select (> _progress 0.75) "/Figment_S09/Athena/Environments/Nexus/FX/P_CattusDoggus_Destabilize3.P_CattusDoggus_Destabilize3" (select (> _progress 0.4) "/Figment_S09/Athena/Environments/Nexus/FX/P_CattusDoggus_Destabilize2.P_CattusDoggus_Destabilize2" "/Figment_S09/Athena/Environments/Nexus/FX/P_CattusDoggus_Destabilize.P_CattusDoggus_Destabilize")))
  (Audio|Parameter|SetFloatParameter (Variables|Default|GetAmbientAudio) "Progress" (Variables|Default|GetProgress) :!class /Script/Engine.AudioComponent))

(fn ConstructionScript
  (CallFunction|UpdateProgress)
  (Rendering|Material|SetScalarParameterValueonMaterials (Variables|Default|GetWave2) "TimeMult" 1.425))
