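# M_NexTunnel is an unlit translucent material in the Nexus environment, apparently for a tunnel effect that can attach to the camera.
# An AttachtoFace static switch selects between a world position offset from AttachMeshToTheCamera, with a small fixed camera-space shift, and a constant zero vector.
# A second switch of the same name selects full value when attached, otherwise vertex color red.
# A pale pink Color parameter is present, but the file shows no material output connections.
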Material /Figment_S09/Athena/Environments/Nexus/Materials/M_NexTunnel
graph {
 properties {blendMode=BLEND_Translucent, pixelDepthOffsetMode=PDOM_Legacy, shadingModel=MSM_Unlit, usedShadingModels=Unlit}
 nodes [
  CameraPositionWS()
  ObjectPositionWS()
  StaticSwitchParameter(
   True=MaterialFunctionCall!(Object Basis 1, Object Basis 2, Object Basis 3, Camera Space Position Offset=Constant3Vector() {constant=<rgba 0 20 0 0>}) {materialFunction=/Engine/Functions/Engine_MaterialFunctions02/WorldPositionOffset/AttachMeshToTheCamera}.World Position Offset
   False=Constant3Vector()
  ) {parameterName=AttachtoFace}
  StaticSwitchParameter(True=Constant() {r=1}, False=VertexColor().R) {parameterName=AttachtoFace}
  VectorParameter() {parameterName=Color, defaultValue=<rgba 1 0.7298 0.8539 1>}
  WorldPosition()
 ]
}
